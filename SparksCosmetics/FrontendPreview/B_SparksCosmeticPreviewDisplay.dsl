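; Frontend preview actor that displays Sparks instrument cosmetics: guitar, bass, keytar, mic with stand, or drums with sticks.
; On customize, it stores the item definition and maps the subtype to main, left-hand and auxiliary mesh components.
; A rollback CVar at zero selects the legacy Sparks path, which sets variant meshes, materials, frontend-flagged FX and LUT parameters per component.
; Otherwise it applies the shared BR variant flow to those meshes and shows them.
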(event Visuals|EventCustomizePreviewInstance
  (params (InItemDefinition :type Fort%20Item%20Definition%20Object%20Reference))
  (bind _assparks_instrument_item_def_base (Utilities|Casting|CastToSparksInstrumentItemDefBase InItemDefinition))
  (CallFunction|HideAll)
  (Variables|SparksInstrumentPreviewActor|SetItemDef (Utilities|ResolveSoftReference _assparks_instrument_item_def_base))
  ; Check for the CVar to see how we want to decorate this
  (if (== (Fortnite|GetNumericCVar "Fort.Rollback.UseBRCosmeticsFlowForSparksInstruments") 0)
    ; use old Sparks decoration code
    (bind (_mainmesh _lhmesh _auxmesh) (CallFunction|GetMeshestoUse (Variables|/Script/SparksCoreCosmeticsRuntime.SparksAccountItemDefinition|GetSubType _assparks_instrument_item_def_base)))
    (CallFunction|FixupMeshesandMaterials _mainmesh _lhmesh _auxmesh _assparks_instrument_item_def_base)
    (else
      ; Use the same variant decoration code as other BR items
      (bind _meshes (CallFunction|GetMeshArray (Variables|/Script/SparksCoreCosmeticsRuntime.SparksAccountItemDefinition|GetSubType _assparks_instrument_item_def_base)))
      (CallFunction|ApplyVariantstoMeshes _assparks_instrument_item_def_base _meshes)
      (for _array_element _meshes
        (Rendering|SetVisibility _array_element true true)))))

(fn FixupAUXMeshAndMaterials
  (params (Mesh :type Skeletal%20Mesh%20Component%20Object%20Reference) (ItemDefinition :type Sparks%20Instrument%20Item%20Def%20Base%20Object%20Reference) (Choices :type Array%20of%20Mcp%20Variant%20Channel%20Info%20Structures))
  (Utilities|IsValid Mesh
    (:Is%20Valid
      (Components|SkeletalMesh|SetSkeletalMeshAsset Mesh (Class|SparksInstrumentItemDefBase|GetAuxiliaryMesh ItemDefinition Choices :!class /Script/SparksCoreCosmeticsRuntime.SparksInstrumentItemDefBase))
      (Class|SparksInstrumentItemDefBase|AttachAUXMaterials ItemDefinition Choices Mesh)
      (bind _returnvalue (Class|SparksInstrumentItemDefBase|SpawnandAttachAuxFXSystem ItemDefinition Choices Mesh true))
      (bind _output_get (LocalVariables|SetVFX _returnvalue))
      (Effects|Components|ParticleSystem|SetBoolParameter _output_get "IsFrontEnd" true)
      (Effects|Components|ParticleSystem|SetBoolParameter _output_get "IsFrontEndPreview" true)
      (Rendering|SetVisibility Mesh true true))))

(fn FixupLHMeshAndMaterials
  (params (Mesh :type Skeletal%20Mesh%20Component%20Object%20Reference) (ItemDefinition :type Sparks%20Instrument%20Item%20Def%20Base%20Object%20Reference) (Choices :type Array%20of%20Mcp%20Variant%20Channel%20Info%20Structures))
  (Utilities|IsValid Mesh
    (:Is%20Valid
      (Components|SkeletalMesh|SetSkeletalMeshAsset Mesh (Class|SparksInstrumentItemDefBase|GetLeftHandMesh ItemDefinition Choices :!class /Script/SparksCoreCosmeticsRuntime.SparksInstrumentItemDefBase))
      (Class|SparksInstrumentItemDefBase|AttachLHMaterials ItemDefinition Choices Mesh)
      (bind _returnvalue (Class|SparksInstrumentItemDefBase|SpawnandAttachLHFXSystem ItemDefinition Choices Mesh true))
      (bind _output_get (LocalVariables|SetVFX _returnvalue))
      (Effects|Components|ParticleSystem|SetBoolParameter _output_get "IsFrontEnd" true)
      (Effects|Components|ParticleSystem|SetBoolParameter _output_get "IsFrontEndPreview" true)
      (Rendering|SetVisibility Mesh true true))))

(fn FixupMainMeshAndMaterials
  (params (Mesh :type Skeletal%20Mesh%20Component%20Object%20Reference) (ItemDefinition :type Sparks%20Instrument%20Item%20Def%20Base%20Object%20Reference) (Choices :type Array%20of%20Mcp%20Variant%20Channel%20Info%20Structures))
  (Utilities|IsValid Mesh
    (:Is%20Valid
      (Components|SkeletalMesh|SetSkeletalMeshAsset Mesh (Class|SparksInstrumentItemDefBase|GetMesh ItemDefinition Choices :!class /Script/SparksCoreCosmeticsRuntime.SparksInstrumentItemDefBase))
      (Class|SparksInstrumentItemDefBase|AttachMainMaterials ItemDefinition Choices Mesh)
      (bind _returnvalue (Class|SparksInstrumentItemDefBase|SpawnandAttachMainFXSystem ItemDefinition Choices Mesh true))
      (bind _output_get (LocalVariables|SetVFX _returnvalue))
      (Effects|Components|ParticleSystem|SetBoolParameter _output_get "IsFrontEnd" true)
      (Effects|Components|ParticleSystem|SetBoolParameter _output_get "IsFrontEndPreview" true)
      (Rendering|SetVisibility Mesh true true))))

(fn FixupMeshAndMaterials
  (params (MeshComponent :type Skeletal%20Mesh%20Component%20Object%20Reference) (NewMesh :type Skeletal%20Mesh%20Object%20Reference) (Material1 :type Material%20Instance%20Object%20Reference) (Material2 :type Material%20Instance%20Object%20Reference) (FX :type FXSystem%20Asset%20Object%20Reference) (FX_Socket :type Name))
  (Utilities|IsValid MeshComponent
    (:Is%20Valid
      (Utilities|FlowControl|Sequence
        (:then_0
          (Utilities|IsValid NewMesh
            (:Is%20Valid
              (Components|SkeletalMesh|SetSkeletalMeshAsset MeshComponent NewMesh))))
        (:then_1
          (Utilities|IsValid Material1
            (:Is%20Valid
              (Rendering|Material|SetMaterial MeshComponent 0 Material1))))
        (:then_2
          (Utilities|IsValid Material2
            (:Is%20Valid
              (Rendering|Material|SetMaterial MeshComponent 1 Material2))))
        (:then_3
          (Utilities|IsValid FX
            (:Is%20Valid
              (bind _returnvalue (Effects|SpawnFXSystemAttached FX MeshComponent FX_Socket))
              (Effects|Components|ParticleSystem|SetBoolParameter _returnvalue "IsFrontEnd" true))))
        (:then_4
          (Rendering|SetVisibility MeshComponent true true))))))

(fn FixupMeshesAndMaterials
  (params (MainMesh :type Skeletal%20Mesh%20Component%20Object%20Reference) (LHMesh :type Skeletal%20Mesh%20Component%20Object%20Reference) (AUXMesh :type Skeletal%20Mesh%20Component%20Object%20Reference) (ItemDefinition :type Sparks%20Instrument%20Item%20Def%20Base%20Object%20Reference))
  (bind _returnvalue (CallFunction|GetCurrentChoices))
  (bind _auxmesh (Variables|GetAUXMesh))
  (bind _itemdefinition (Variables|GetItemDefinition))
  (bind _itemdefinition_2 (Variables|GetItemDefinition))
  (bind _itemdefinition_3 (Variables|GetItemDefinition))
  (bind _mainmesh (Variables|GetMainMesh))
  (bind _lhmesh (Variables|GetLHMesh))
  (Utilities|FlowControl|Sequence
    (:then_0
      (Utilities|IsValid _mainmesh
        (:Is%20Valid
          (CallFunction|FixupMainMeshandMaterials _mainmesh _itemdefinition_2 _returnvalue)
          (Class|SparksInstrumentItemDefBase|ApplyFrontendLUTParamstoMaterials _itemdefinition_2 _mainmesh))))
    (:then_1
      (Utilities|IsValid _lhmesh
        (:Is%20Valid
          (CallFunction|FixupLHMeshandMaterials _lhmesh _itemdefinition_3 _returnvalue)
          (Class|SparksInstrumentItemDefBase|ApplyFrontendLUTParamstoMaterials _itemdefinition_3 _lhmesh))))
    (:then_2
      (Utilities|IsValid _auxmesh
        (:Is%20Valid
          (CallFunction|FixupAUXMeshandMaterials _auxmesh _itemdefinition _returnvalue)
          (Class|SparksInstrumentItemDefBase|ApplyFrontendLUTParamstoMaterials _itemdefinition _auxmesh))))))

(fn GetMeshArray
  (params (NewParam :type ESparksAccountItemSubtype%20Enum :default "Aura"))
  (bind _meshestouse (LocalVariables|GetMeshesToUse))
  (bind (_mainmesh _lhmesh _auxmesh) (CallFunction|GetMeshestoUse NewParam))
  (Utilities|IsValid _mainmesh
    (:Is%20Valid
      (Utilities|Array|Add _meshestouse _mainmesh))
    (:Is%20Not%20Valid
      _))
  (Utilities|IsValid _lhmesh
    (:Is%20Valid
      (Utilities|Array|Add _meshestouse _lhmesh))
    (:Is%20Not%20Valid
      _))
  (Utilities|IsValid _auxmesh
    (:Is%20Valid
      (Utilities|Array|Add _meshestouse _auxmesh))
    (:Is%20Not%20Valid
      _))
  (return (LocalVariables|GetMeshesToUse)))

(fn GetMeshesToUse
  (params (SparksType :type ESparksAccountItemSubtype%20Enum :default "Aura"))
  (switch Utilities|FlowControl|Switch|SwitchonESparksAccountItemSubtype SparksType
    (:Guitar
      (return (Variables|Default|GetGuitarMesh) 0 0))
    (:Bass
      (return (Variables|Default|GetBassMesh) 0 0))
    (:Keyboard
      (return (Variables|Default|GetKeytarMesh) 0 0))
    (:Microphone
      (return (Variables|Default|GetMicMesh) (Variables|Default|GetMicStandMesh) 0))
    (:Drums
      (return (Variables|Default|GetRHDrumstick) (Variables|Default|GetLHDrumstick) (Variables|Default|GetDrumMesh)))))

(fn Hideall
  (Rendering|SetVisibility (Variables|Default|GetMicStandMesh) false true))

(fn ConstructionScript)
