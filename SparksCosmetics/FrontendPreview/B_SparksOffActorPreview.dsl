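; Frontend off-pawn preview actor for Sparks instrument cosmetics, handling floor, lighting and the reveal transition.
; On spawn it toggles a floor plane by instrument type and display settings, then sets up lights, binding their refresh to game settings changes.
; Lighting switches between PC rim and mobile directional lights by platform, performance mode and shadow quality.
; When shown, timelines reveal the preview with Niagara, sound and a material dissolve on the instrument meshes.
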(event EventOnPreviewVisualsBeginLoading
  (CallFunction|HidePreview))

(event EventOnPreviewVisualsSpawned
  (Utilities|FlowControl|Sequence
    (:then_0
      (bind _returnvalue (CallFunction|GetInstrumentType))
      (bind _output_get (Variables|Default|SetShow%20Floor (and (Utilities|Array|ContainsItem (Variables|Default|GetInstrument%20Types%20with%20Floor) _returnvalue) (Utilities|Struct|BreakCharacterDisplaySettings (Variables|FortItemPreviewActor|GetCurrentDisplayInfo)))))
      (Development|SetHiddeninGame (Variables|Default|GetPlane) (not _output_get) true))
    (:then_1
      (Default|SetUPLighting)
      (Default|LightControl true))
    (:then_2
      (bind _returnvalue_2 (Settings|GetGameUserSettings))
      (EventDispatchers|BindEventtoOnGameUserSettingsUINeedsUpdate _returnvalue_2 (delegate Custom|UpdateSettings)))))

(event Custom|UpdateSettings
  (Default|HandleLightingSettingsChanged))

(event EventOnPreviewVisualsShown
  (bind _skeletalmeshtoresin (Variables|Default|GetSkeletalMeshToResIn))
  (Timeline|Timeline PlayFromStart
    (:ShowPawn
      (CallFunction|ShowPreview))
    (:PlayResIn
      (Niagara|SpawnSystematLocation "/Game/Effects/Fort_Effects/UI/Transition/P_FrontEndCharacterTransition_Converted.P_FrontEndCharacterTransition_Converted" (Transformation|GetWorldLocation (Variables|FortItemPreviewOffPawnActor|GetPreviewActorComponent)))
      (CallFunction|SpawnSoundPlayback)
      (CallFunction|GetInstrumentMeshComponents)
      ;[1 1 1] Custom Data needs setting up for this
      (bind (_direction _hiddenamount) (Timeline|Timeline_0 PlayFromStart)
        (:Update
          (for _array_element _skeletalmeshtoresin
            (Rendering|Material|SetScalarParameterValueonMaterials _array_element "NewDissolveGradient" _hiddenamount)))
        (:Finished
          (Utilities|Array|Clear _skeletalmeshtoresin))))) ; Sequence Show / Outro / ResIn
)

(event Custom|SpawnSoundPlayback
  ;[0 0 0] Resin sound playback
  (Audio|PlaySound2D (Variables|Default|GetResIn%20Sound)))

(fn GetInstrumentMeshComponents
  (bind _skeletalmeshtoresin (Variables|Default|GetSkeletalMeshToResIn))
  (Utilities|Array|Clear (Variables|Default|GetSkeletalMeshToResIn))
  (bind _asb_sparks_cosmetic_preview_display (Utilities|Casting|CastToB_SparksCosmeticPreviewDisplay (Variables|/Script/Engine.ChildActorComponent|GetChildActor (Variables|FortItemPreviewOffPawnActor|GetPreviewActorComponent))))
  (bind _returnvalue (CallFunction|GetInstrumentType))
  (switch Utilities|FlowControl|Switch|SwitchonESparksInstrumentType _returnvalue
    (:Guitar
      _)
    (:Bass
      _)
    (:Vocals
      (Utilities|Array|Add _skeletalmeshtoresin (Variables|/SparksCosmetics/FrontendPreview/B_SparksCosmeticPreviewDisplay.B_SparksCosmeticPreviewDisplay_C|GetMicMesh _asb_sparks_cosmetic_preview_display))
      (Utilities|Array|Add _skeletalmeshtoresin (Variables|/SparksCosmetics/FrontendPreview/B_SparksCosmeticPreviewDisplay.B_SparksCosmeticPreviewDisplay_C|GetMicStandMesh _asb_sparks_cosmetic_preview_display)))
    (:Drum
      (Utilities|Array|Add _skeletalmeshtoresin (Variables|/SparksCosmetics/FrontendPreview/B_SparksCosmeticPreviewDisplay.B_SparksCosmeticPreviewDisplay_C|GetLHDrumstick _asb_sparks_cosmetic_preview_display))
      (Utilities|Array|Add _skeletalmeshtoresin (Variables|/SparksCosmetics/FrontendPreview/B_SparksCosmeticPreviewDisplay.B_SparksCosmeticPreviewDisplay_C|GetRHDrumstick _asb_sparks_cosmetic_preview_display))
      (Utilities|Array|Add _skeletalmeshtoresin (Variables|/SparksCosmetics/FrontendPreview/B_SparksCosmeticPreviewDisplay.B_SparksCosmeticPreviewDisplay_C|GetDrumMesh _asb_sparks_cosmetic_preview_display)))
    (:Keyboard
      _))
  (Utilities|Array|Add _skeletalmeshtoresin (Variables|/SparksCosmetics/FrontendPreview/B_SparksCosmeticPreviewDisplay.B_SparksCosmeticPreviewDisplay_C|GetGuitarMesh _asb_sparks_cosmetic_preview_display)))

(fn HandleLightingSettingsChanged
  (bind _returnvalue (TouchPlatform|IsMobileGame))
  (bind _returnvalue_2 (Settings|GetFortGameUserSettings))
  (bind _returnvalue_3 (Class|FortGameUserSettings|IsinPerformanceMode _returnvalue_2))
  (bind _returnvalue_4 (Math|Boolean|ORBoolean _returnvalue _returnvalue_3 false))
  (bind _returnvalue_5 (Settings|GetGameUserSettings))
  (if (Variables|Default|GetIs%20%20active)
    (if (or (== (Settings|GetShadowQuality _returnvalue_5) 0) _returnvalue_4)
      (if _returnvalue_4
        (Default|SwitchPCLighting false)
        (Default|SwitchMobileLighting true)
        (else
          (Default|SwitchPCLighting true)))
      (else
        (Default|SwitchMobileLighting false)
        (Default|SwitchPCLighting true)))))

(fn LightControl
  (params (Active :type Boolean))
  (bind _returnvalue (TouchPlatform|IsMobileGame))
  (bind _returnvalue_2 (TouchPlatform|IsNintendoSwitch))
  (bind _returnvalue_3 (TouchPlatform|IsNintendoSwitch))
  (bind _returnvalue_4 (Settings|GetFortGameUserSettings))
  (bind _returnvalue_5 (Class|FortGameUserSettings|IsinPerformanceMode _returnvalue_4))
  (bind _returnvalue_6 (Math|Boolean|ORBoolean _returnvalue _returnvalue_5 _returnvalue_2))
  (bind _is__active (Variables|Default|GetIs%20%20active))
  (Variables|Default|SetIs%20%20active Active)
  (bind _returnvalue_7 (Settings|GetGameUserSettings))
  (if (Math|Boolean|ORBoolean (== (Settings|GetShadowQuality _returnvalue_7) 0) _returnvalue_6 _returnvalue_3)
    (if _returnvalue_6
      (Default|SwitchMobileLighting _is__active)
      (elif _returnvalue_3
        (Default|SwitchErebusLighting _returnvalue_3))))
  (Default|SwitchPCLighting _is__active))

(fn SetUPLighting
  (Rendering|Components|Light|SetForceCachedShadowsforMovablePrimitives (Variables|Default|GetRimTop_pc) true)
  (Default|SwitchMobileLighting false)
  (Default|SwitchPCLighting false))

(fn SwitchErebusLighting
  (params (visibility :type Boolean))
  (Rendering|SetVisibility (Variables|Default|GetRimTop_pc) false false)
  (Rendering|SetVisibility (Variables|Default|GetDirectionalLight_mobile) visibility false))

(fn SwitchMobileLighting
  (params (NewParam :type Boolean))
  (Rendering|SetVisibility (Variables|Default|GetRimTop_pc) false false)
  (Rendering|SetVisibility (Variables|Default|GetDirectionalLight_mobile) NewParam false))

(fn SwitchPCLighting
  (params (Visibility :type Boolean))
  (Rendering|SetVisibility (Variables|Default|GetRimTop_pc) Visibility false)
  (Rendering|SetVisibility (Variables|Default|GetDirectionalLight_mobile) false false))

(fn ConstructionScript)
