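; Frontend preview actor for Sparks cosmetics shown on a player pawn, managing lighting, floor and res-in presentation.
; Loading sets toon light parameters in a global collection and, unless AlwaysOn, switches light rigs by platform and graphics settings.
; The floor hides for gliders and contrails and takes custom, default or autotest materials.
; Showing plays an optionally CVar-delayed Niagara, sound and two-stage dissolve, then reapplies instrument bone hiding.
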(event EventOnPreviewVisualsBeginLoading
  (Utilities|FlowControl|Sequence
    (:then_0
      (if (not (Variables|Default|GetAlwaysOn))
        (CallFunction|SetupLighting)
        (bind _returnvalue (Settings|GetGameUserSettings))
        (EventDispatchers|BindEventtoOnGameUserSettingsUINeedsUpdate _returnvalue (EventDispatchers|CreateEvent :!event HandleLightingSettingsChanged))))
    (:then_1
      ;[1 1 1] Setting Fake Light Vector For Toon Characters
      (Rendering|Material|SetVectorParameterValue "/Game/Packages/Fortress_SharedMaterials/GlobalMaterialParameters/FortniteMaterialParameters.FortniteMaterialParameters" "SunlightDirectionalVector" (Math|Conversions|ToLinearColor(Vector) (Math|Vector|RotateVector (Transformation|GetActorForwardVector) (Variables|Default|GetToonLightRotatio))) :!class /Script/Engine.KismetMaterialLibrary)
      (Rendering|Material|SetScalarParameterValue "/Game/Packages/Fortress_SharedMaterials/GlobalMaterialParameters/FortniteMaterialParameters.FortniteMaterialParameters" "bEnableCharacterRimLight" :!class /Script/Engine.KismetMaterialLibrary))))

(event EventOnPreviewVisualsSpawned
  (bind (_primaryitem _bshowfloor) (Utilities|Struct|BreakCharacterDisplaySettings (Variables|FortItemPreviewActor|GetCurrentDisplayInfo)))
  (bind _output_get (Variables|Default|SetShow%20Floor (and (Math|Boolean|NORBoolean (Utilities|IsA _primaryitem "/Script/FortniteGame.AthenaGliderItemDefinition") (Utilities|IsA _primaryitem "/Script/FortniteGame.AthenaSkyDiveContrailItemDefinition")) _bshowfloor)))
  (CallFunction|SetFloorEnabled _output_get))

(event EventOnPreviewVisualsShown
  (bind (_outvalue _returnvalue) (Fortnite|GetNumericCVar "FX.IASResinDelay"))
  (if (and (> _outvalue 0) _returnvalue)
    (bind _returnvalue_2 (Utilities|Time|SetTimerbyFunctionName 0 "SpawnResInEffects" _outvalue))
    (Variables|Default|SetIASResinDelayTimer _returnvalue_2)
    (else
      (CallFunction|SpawnResinEffects))))

(event EventOnCurrentVisualsCleanedUp
  (Utilities|Time|ClearandInvalidateTimerbyHandle (Variables|Default|GetIASResinDelayTimer)))

(event EventOnSetFloorMaterial
  (params (InMaterialInstance :type Material%20Interface%20Object%20Reference))
  ;[0 0 0] OnSetFloorMaterial
  (Variables|Default|SetCustomFloorMaterial InMaterialInstance)
  (Rendering|Material|SetMaterial (Variables|Default|GetPlane) 0 (select (Utilities|IsValid (Variables|Default|GetCustomFloorMaterial) :!class /Script/Engine.KismetSystemLibrary) (Variables|Default|GetCustomFloorMaterial) (Variables|Default|GetDefaultFloorMaterial))))

(event EventOnUpdateFloorMaterial
  (params (bEnableAutotestBackground :type Boolean))
  (if bEnableAutotestBackground
    (bind _returnvalue (Rendering|Material|CreateDynamicMaterialInstance (Variables|FortItemPreviewMultiAngleActor|GetAutotestFloorMaterial) "None" "None" :!class /Script/Engine.KismetMaterialLibrary))
    (Rendering|Material|SetMaterial (Variables|Default|GetPlane) 0 _returnvalue)
    (else
      (Rendering|Material|SetMaterial (Variables|Default|GetPlane) 0 (select (Utilities|IsValid (Variables|Default|GetCustomFloorMaterial) :!class /Script/Engine.KismetSystemLibrary) (Variables|Default|GetCustomFloorMaterial) (Variables|Default|GetDefaultFloorMaterial))))))

(event Custom|SpawnSoundPlayback
  ;[0 0 0] Resin sound playback
  (Audio|PlaySoundatLocation "/Game/Sounds/Fort_UI_Sounds/S13_Frontend/BattlePass_ResIn_Cue.BattlePass_ResIn_Cue" (CallFunction|GetPawnLocation)))

(event Custom|RezInMaterialEffect
  ;[1 1 1] RezInMaterialEffect
  (bind _returnvalue (Rendering|Material|CreateDynamicMaterialInstance "/Game/Effects/Fort_Effects/UI/Transition_IAS/ChaDeRezShader/CharacterDeRezShader_frontend_Inst.CharacterDeRezShader_frontend_Inst" "None" "None" :!class /Script/Engine.KismetMaterialLibrary))
  (Rendering|Material|SetScalarParameterValue _returnvalue "DissolveGradientPercentage" 0.0 :!class /Script/Engine.MaterialInstanceDynamic)
  (bind _asfort_player_pawn (Utilities|Casting|CastToFortPlayerPawn (Variables|/Script/Engine.ChildActorComponent|GetChildActor (Variables|FortItemPreviewOnPawnActor|GetPrimaryPreviewPawnComponent))))
  (bind _returnvalue_2 (Class|FortPlayerPawn|ApplyMaterialOverrideLocalwithFlags :self _asfort_player_pawn :Material (Utilities|ResolveSoftReference _returnvalue) :Priority 0.0 :Flags 0))
  (bind (_direction _newtrack_0) (Timeline|RezInMaterialEffectTimeLine PlayFromStart)
    (:Update
      (Rendering|Material|SetScalarParameterValue _returnvalue "DissolveGradientPercentage" _newtrack_0 :!class /Script/Engine.MaterialInstanceDynamic)
      (Cosmetics|SetCosmeticModifierChildParticleComponentsVisible _asfort_player_pawn))
    (:Finished
      (Class|FortPlayerPawn|RemoveMaterialOverrideLocal _asfort_player_pawn _returnvalue_2)
      (bind (_direction_2 _newtrack_0_2) (Timeline|RezInMaterialEffectTimeLine2 PlayFromStart)
        (:Update
          (bind _skinned_mesh_component (Utilities|ForEachSkinnedMeshComponentonPawn _asfort_player_pawn)
            (:Loop%20Body
              (Rendering|Material|SetScalarParameterValueonMaterials _skinned_mesh_component "NewDissolveGradient" _newtrack_0_2)
              (Rendering|Material|SetScalarParameterValueonMaterials _skinned_mesh_component "UsePannerAnimation" 1.0))))
        (:Finished
          (bind _skinned_mesh_component_2 (Utilities|ForEachSkinnedMeshComponentonPawn _asfort_player_pawn)
            (:Loop%20Body
              (Rendering|Material|SetScalarParameterValueonMaterials _skinned_mesh_component_2 "UsePannerAnimation" 0.0))
            (:Loop%20Completed
              (Utilities|SetPartModifierChildParticleComponentsVisibleonPawn _asfort_player_pawn true)
              ;;[0.15 0.07268 0.000519 0.5] FORT-902675 - Fix for VFX appearing while having the same Backbling as Instrument
              (Utilities|IsValid _asfort_player_pawn
                (:Is%20Valid
                  (bind _assparks_instrument_item_def_base (Utilities|Casting|CastToSparksInstrumentItemDefBase (CallFunction|GetItemDefinition)))
                  (Class|FortPlayerPawn|RefreshBoneSetHiding _asfort_player_pawn (Class|SparksInstrumentItemDefBase|GetBoneSetstoHide _assparks_instrument_item_def_base) true))))))))))

(fn GetLODStreamingSafetyDuration)

(fn HandleLightingSettingsChanged
  (bind _returnvalue (TouchPlatform|IsNintendoSwitch))
  (bind _returnvalue_2 (TouchPlatform|IsMobileGame))
  (bind _returnvalue_3 (Settings|GetFortGameUserSettings))
  (bind _returnvalue_4 (Class|FortGameUserSettings|IsinPerformanceMode _returnvalue_3))
  (bind _returnvalue_5 (Math|Boolean|ORBoolean _returnvalue_2 _returnvalue_4 _returnvalue))
  ;[0 0 0 1 12] Switch lighting if settings are changed while screen is already active
  (bind _returnvalue_6 (Settings|GetGameUserSettings))
  (if (Variables|Default|GetIsActive)
    (if (or (== (Settings|GetShadowQuality _returnvalue_6) 0) _returnvalue_5)
      (if _returnvalue_5
        ;;[0.125 0.125 0.125 1 12] Enable Mobile
        (CallFunction|SwitchPCLighting false)
        (CallFunction|SwitchMobileLighting true)
        (else
          ;;[0.125 0.125 0.125 1 12] Enable Low Detail PC
          (CallFunction|SwitchPCLighting true)))
      (else
        ;;[0.125 0.125 0.125 1 12] Enable HighEnd
        (CallFunction|SwitchMobileLighting false)
        (CallFunction|SwitchPCLighting true)))))

(fn IsSkyDiveContrailItem
  (bind (_asathena_sky_dive_contrail_item_definition _bsuccess) (Utilities|Casting|CastToAthenaSkyDiveContrailItemDefinition (Utilities|Struct|BreakCharacterDisplaySettings (Variables|FortItemPreviewActor|GetCurrentDisplayInfo))))
  (return _bsuccess))

(fn LightControl
  (params (Active :type Boolean))
  (bind _returnvalue (TouchPlatform|IsMobileGame))
  (bind _returnvalue_2 (Settings|GetFortGameUserSettings))
  (bind _returnvalue_3 (Class|FortGameUserSettings|IsinPerformanceMode _returnvalue_2))
  (bind _returnvalue_4 (TouchPlatform|IsNintendoSwitch))
  (bind _returnvalue_5 (TouchPlatform|IsNintendoSwitch))
  (bind _returnvalue_6 (Math|Boolean|ORBoolean _returnvalue _returnvalue_3 _returnvalue_5))
  (bind _isactive (Variables|Default|GetIsActive))
  (if (not (Variables|Default|GetAlwaysOn))
    (Variables|Default|SetIsActive Active)
    (bind _returnvalue_7 (Settings|GetGameUserSettings))
    (if (Math|Boolean|ORBoolean (== (Settings|GetShadowQuality _returnvalue_7) 0) _returnvalue_6 _returnvalue_4)
      (if _returnvalue_6
        (CallFunction|SwitchMobileLighting _isactive)
        (elif _returnvalue_4
          (CallFunction|SwitchErebusLighting _isactive))))
    (CallFunction|SwitchPCLighting (Variables|Default|GetIsActive))))

(fn SetFloorEnabled
  (params (Show%20Floor :type Boolean))
  (Development|SetHiddeninGame (Variables|Default|GetPlane) (not Show%20Floor) true))

(fn SetupLighting
  (if (not (Variables|Default|GetAlwaysOn))
    (Utilities|FlowControl|Sequence
      (:then_0
        (Rendering|Components|Light|SetForceCachedShadowsforMovablePrimitives (Variables|Default|GetKeyLight5) true))
      (:then_1
        ;[0 0 0 1 12] Forece Disable on start
        (CallFunction|SwitchMobileLighting false)
        (CallFunction|SwitchPCLighting false)
        (CallFunction|LightControl true)))))

(fn SpawnResInEffects
  (Components|DestroyComponent (Variables|Default|GetObscuringLoopEmitter))
  (Niagara|SpawnSystematLocation "/Game/Effects/Fort_Effects/UI/Transition/P_FrontEndCharacterTransition_Converted.P_FrontEndCharacterTransition_Converted" (- (CallFunction|GetPawnLocation) "0.000000,0.000000,80.000000"))
  (CallFunction|RezinMaterialEffect)
  (CallFunction|SpawnSoundPlayback))

(fn SwitchErebusLighting
  (params (Visibility :type Boolean))
  (Rendering|SetVisibility (Variables|Default|GetKeyLight_StandaloneForSwitch) Visibility false))

(fn SwitchMobileLighting
  (params (NewParam :type Boolean))
  (Rendering|SetVisibility (Variables|Default|GetBounce1) false false)
  (Rendering|SetVisibility (Variables|Default|GetDirectionalLightMobile) NewParam false))

(fn SwitchPCLighting
  (params (Visibility :type Boolean))
  (Rendering|SetVisibility (Variables|Default|GetBounce1) Visibility false)
  (Rendering|SetVisibility (Variables|Default|GetKeyLight_StandaloneForSwitch) false false))

(fn ConstructionScript)
